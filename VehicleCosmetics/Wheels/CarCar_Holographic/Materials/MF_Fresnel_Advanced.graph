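# Material function in the holographic CarCar wheel cosmetic folder that builds adjustable Fresnel-style Edge and Middle masks from the view angle.
# It dots the transformed input normal with the camera vector, then scales that value and its inverse by sharpness before clamping and inverting each.
# The Two Sided switch feeds in the saturated one-minus dot instead, which effectively swaps the masks, as its Abs after Saturate has no effect.
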MaterialFunction /VehicleCosmetics/Wheels/CarCar_Holographic/Materials/MF_Fresnel_Advanced
graph {nodes [
  n1 = Constant3Vector() {constant=<rgba 0 0 1 0>}
  n2 = DotProduct(
   A=Transform(None=FunctionInput(Preview=@n1) {bUsePreviewValueAsDefault=true, inputName=Input Normal, previewValue=<0 0 1 1>})
   B=CameraVectorWS()
  )
  n3 = OneMinus(None=@n2)
  n4 = StaticSwitch(
   True=Abs(None=Saturate(None=@n3))
   False=Saturate(None=@n2)
   Value=FunctionInput(Preview=StaticBool()) {bUsePreviewValueAsDefault=true, inputName=Input Two Sided, inputType=FunctionInput_StaticBool, sortPriority=2}
  )
  n5 = FunctionInput(Preview=Constant() {r=1}) {bUsePreviewValueAsDefault=true, inputName=Input Sharpness, inputType=FunctionInput_Scalar, sortPriority=1}
  n6 = Multiply(A=@n4, B=@n5)
  FunctionOutput(None=OneMinus(None=Clamp(None=@n6))) {outputName=Edge}
  n7 = Multiply(A=OneMinus(None=@n4), B=@n5)
  FunctionOutput(None=OneMinus(None=Clamp(None=@n7))) {outputName=Middle, sortPriority=1}
 ]}
