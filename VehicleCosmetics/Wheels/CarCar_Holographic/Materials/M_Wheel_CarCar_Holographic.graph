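# Translucent lit master material for the holographic CarCar wheel, with separate rim and tread controls for opacity, normal, metallic and roughness.
# Panning gradient textures, shaped by masks, drive emissive pulses: the rim on UV1, the tread blending straight and rotated scroll directions.
# Fresnel edge glow and color parameters are also declared, linked through named reroutes.
# The attributes pass through MF_BaseVehicle_Damage, then MF_VehicleDeRez_Transition.
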Material /VehicleCosmetics/Wheels/CarCar_Holographic/Materials/M_Wheel_CarCar_Holographic
graph {
 properties {
  bAutomaticallySetUsageInEditor=false
  bUseMaterialAttributes=true
  bUsedWithMeshDeformer=true
  bUsedWithSkeletalMesh=true
  blendMode=BLEND_Translucent
  translucencyLightingMode=TLM_SurfacePerPixelLighting
  usedShadingModels=Default Lit
 }
 nodes [
  n1 = NamedRerouteUsage()
  n2 = SetMaterialAttributes!(
   MaterialAttributes
   Base Color=NamedRerouteUsage()
   Metallic=NamedRerouteUsage()
   Roughness=NamedRerouteUsage()
   Emissive Color=NamedRerouteUsage()
   Opacity=@n1
   Opacity Mask=@n1
   Normal=NamedRerouteUsage()
  ) {attributeSetTypes [
    #1
    #2
    #3
    #4
    #5
    #6
    #7
   ]}
  MaterialFunctionCall(In=MaterialFunctionCall(Material=@n2) {materialFunction=&MF_BaseVehicle_Damage}.Result) {materialFunction=&MF_VehicleDeRez_Transition}
  NamedRerouteDeclaration(None=VertexColor().R) {name=VC_R, nodeColor=<rgba 0.1016 1 0 1>}
  n3 = TextureObjectParameter() {group=0_Base, parameterName=Gradient, sortPriority=2, texture=/VehicleCosmetics/Wheels/CarCar_Holographic/Textures/Gradiant_V}
  NamedRerouteDeclaration(None=@n3) {name=Gradient, nodeColor=<rgba 0 1 0.1797 1>}
  NamedRerouteDeclaration(None=VectorParameter() {group=0_Base, parameterName=Base Color, sortPriority=0, defaultValue=<rgba 0 0 0 1>}.RGB) {name=Base Color, nodeColor=<rgba 0 1 0.625 1>}
  n4 = ScalarParameter() {defaultValue=1, group=1_Rim, parameterName=Rim Opacity}
  n5 = NamedRerouteUsage()
  NamedRerouteDeclaration(None=LinearInterpolate(A=Add(A=@n4, B=@n5), B=ScalarParameter() {defaultValue=0.5, group=2_Tread, parameterName=Tread Opacity}, Alpha=NamedRerouteUsage())) {name=Opacity, nodeColor=<rgba 1 0.3047 0 1>}
  n6 = TextureSampleParameter2D() {
   group=0_Base
   parameterName=Normal Rim
   samplerType=SAMPLERTYPE_Normal
   sortPriority=4
   texture=/VehicleCosmetics/Wheels/CarCar_Holographic/Textures/T_CarCar_Rim_N
  }
  n7 = TextureSampleParameter2D() {
   group=0_Base
   parameterName=Normal Tread
   samplerType=SAMPLERTYPE_Normal
   sortPriority=5
   texture=/VehicleCosmetics/Wheels/CarCar_Holographic/Textures/T_CarCar_Tire_N
  }
  NamedRerouteDeclaration(None=LinearInterpolate(A=@n6.RGB, B=@n7.RGB, Alpha=NamedRerouteUsage())) {name=Normal, nodeColor=<rgba 0.5 0.5 1 1>}
  n8 = ScalarParameter() {defaultValue=1.5, group=3_Fresnel, parameterName=Fresnel Intensity}
  n9 = MaterialFunctionCall!(
   Input Normal=NamedRerouteUsage()
   Input Sharpness=ScalarParameter() {defaultValue=1.3, group=3_Fresnel, parameterName=Fresnel Sharpness}
   Input Two Sided
  ) {materialFunction=&/MF_Fresnel_Advanced}
  NamedRerouteDeclaration(None=Add(A=Multiply(A=@n8, B=@n9.Edge), B=NamedRerouteUsage())) {name=Fresnel, nodeColor=<rgba 1 0 0.6562 1>}
  n10 = TextureCoordinate() {coordinateIndex=1}
  n11 = Panner(Coordinate=@n10) {speedY=-0.5}
  n12 = NamedRerouteUsage()
  n13 = NamedRerouteUsage()
  n14 = Multiply(
   A=ScalarParameter() {defaultValue=1, group=1_Rim, parameterName=Rim Pulse Intensity}
   B=Multiply(
    A=TextureSample(UVs=@n11, Tex=@n12) {samplerSource=SSM_Wrap_WorldGroupSettings, texture=&Gradiant_V}.R
    B=OneMinus(None=@n13)
   )
  )
  NamedRerouteDeclaration(None=Multiply(A=@n14, B=ScalarParameter() {defaultValue=0.3, group=1_Rim, parameterName=Rim Pulse Desaturate})) {name=Rim Pulse Desaturate, nodeColor=<rgba 0.7344 1 0 1>}
  n15 = Multiply(
   A=NamedRerouteUsage()
   B=Multiply(A=NamedRerouteUsage(), B=ScalarParameter() {defaultValue=0.5, group=0_Base, parameterName=EmissiveIntensity})
  )
  n16 = Multiply(A=@n15, B=OneMinus(None=NamedRerouteUsage()))
  n17 = NamedRerouteUsage()
  n18 = NamedRerouteUsage()
  n19 = NamedRerouteUsage()
  n20 = NamedRerouteUsage()
  n21 = Multiply(
   A=NamedRerouteDeclaration(None=LinearInterpolate(A=@n18, B=@n19, Alpha=@n20)) {name=Rim+Tread Pulse, nodeColor=<rgba 0 0.0156 1 1>}
   B=ScalarParameter() {defaultValue=0.5, group=2_Tread, parameterName=Tread Pulse Desaturate}
  )
  NamedRerouteDeclaration(None=LinearInterpolate(A=Add(A=@n16, B=@n17), B=Add(A=@n21, B=@n15), Alpha=NamedRerouteUsage())) {name=Emission, nodeColor=<rgba 1 0.7031 0 1>}
  NamedRerouteDeclaration(None=LinearInterpolate(
    A=ScalarParameter() {group=1_Rim, parameterName=Metallic Rim}
    B=ScalarParameter() {group=2_Tread, parameterName=Metallic Tread}
    Alpha=NamedRerouteUsage()
   )) {name=Metallic, nodeColor=<rgba 0.3828 0 1 1>}
  NamedRerouteDeclaration(None=LinearInterpolate(
    A=ScalarParameter() {defaultValue=0.5, group=1_Rim, parameterName=Roughness Rim}
    B=ScalarParameter() {defaultValue=0.5, group=2_Tread, parameterName=Roughness Tread}
    Alpha=NamedRerouteUsage()
   )) {name=Roughness, nodeColor=<rgba 0 1 0.0625 1>}
  n22 = TextureObjectParameter() {
   group=0_Base
   parameterName=Masks
   samplerType=SAMPLERTYPE_LinearColor
   sortPriority=3
   texture=/VehicleCosmetics/Wheels/CarCar_Holographic/Textures/T_CarCar_Rim_M
  }
  NamedRerouteDeclaration(None=@n22) {name=Masks, nodeColor=<rgba 1 0 0.1406 1>}
  NamedRerouteDeclaration(None=VectorParameter() {group=0_Base, parameterName=CustomColor, sortPriority=1, defaultValue=<rgba 0.8469 0 0 1>}.RGB) {name=Custom Color, nodeColor=<rgba 1 0.8672 0 1>}
  n23 = NamedRerouteUsage()
  n24 = MaterialFunctionCall!(RGBA=TextureSample(Tex=@n23) {samplerType=SAMPLERTYPE_LinearColor, texture=&T_CarCar_Rim_M}.RGBA) {materialFunction=/VehicleCosmetics/SharedMaterials/MaterialFunctions/MF_RGBA_Splitter}
  NamedRerouteDeclaration(None=@n24.R) {name=Masks R, nodeColor=<rgba 0.2656 1 0 1>}
  NamedRerouteDeclaration(None=@n24.G) {name=Masks G, nodeColor=<rgba 0 1 0 1>}
  NamedRerouteDeclaration(None=TextureSample(UVs=@n10, Tex=@n23) {samplerType=SAMPLERTYPE_LinearColor, texture=&T_CarCar_Rim_M}.B) {name=Masks B UV1, nodeColor=<rgba 0 0 1 1>}
  NamedRerouteDeclaration(None=@n24.A) {name=Masks A, nodeColor=<rgba 1 1 1 1>}
  n25 = ScalarParameter() {defaultValue=5, group=0_Base, parameterName=WaveGlowIntensity}
  n26 = NamedRerouteUsage()
  n27 = Panner(Coordinate=MaterialFunctionCall(UVs=TextureCoordinate(), "Rotation Angle (0-1)"=Constant() {r=0.5}) {materialFunction=&CustomRotator}.Rotated Values) {speedX=0.5, speedY=-0.5}
  n28 = NamedRerouteUsage()
  n29 = Panner() {speedY=-0.5}
  n30 = Add(
   A=Multiply(
    A=OneMinus(None=@n26)
    B=TextureSample(UVs=@n27, Tex=@n28) {samplerSource=SSM_Wrap_WorldGroupSettings, texture=&Gradiant_V}.R
   )
   B=Multiply(
    A=@n26
    B=TextureSample(UVs=@n29, Tex=@n28) {samplerSource=SSM_Wrap_WorldGroupSettings, texture=&Gradiant_V}.R
   )
  )
  n31 = NamedRerouteUsage()
  NamedRerouteDeclaration(None=Multiply(A=Multiply(A=@n25, B=@n30), B=TextureSample(Tex=@n31).G)) {name=Tread Pulse, nodeColor=<rgba 0.7344 1 0 1>}
  n32 = ScalarParameter() {defaultValue=1, group=1_Rim, parameterName=Rim Accent Color Intensity}
  n33 = Multiply(A=NamedRerouteUsage(), B=NamedRerouteUsage())
  NamedRerouteDeclaration(None=Add(A=Multiply(A=@n32, B=@n33), B=@n14)) {name=Rim Pulse, nodeColor=<rgba 0.7344 1 0 1>}
 ]
}
